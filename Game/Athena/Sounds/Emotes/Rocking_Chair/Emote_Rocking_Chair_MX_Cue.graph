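# Sound cue for the first-person music of the Rocking Chair emote, combining an intro sound effect with a music intro and loop.
# A mixer starts the slightly attenuated SFX at once, while fixed delays bring in the music intro and then, under a second later, the looping bed.
# An enveloper above the loop's delay defines a very brief volume ramp.
# It uses emote music class and concurrency rules and has a music-analysis submix send.
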SoundCue /Game/Athena/Sounds/Emotes/Rocking_Chair/Emote_Rocking_Chair_MX_Cue
root=Mixer_0
nodes [
 {id=Delay_1, children [WavePlayer_2], properties {delayMax=4.3, delayMin=4.3}}
 {id=Delay_2, children [WavePlayer_0], properties {delayMax=5.02, delayMin=5.02}}
 {
  id=Enveloper_0
  children [Delay_2]
  properties {pitchCurve {editorCurveData {keys [{time=0}]}}, volumeCurve {editorCurveData {keys [<at 0.0012 0>,<at 0.01 1>]}}}
 }
 {id=Mixer_0, children [WavePlayer_1,Delay_1,Enveloper_0], properties {inputVolume [0.7,1,1]}}
 {id=WavePlayer_0, properties {bLooping=true, soundWaveAssetPtr=&/Emote_MX_RockingChair_Loop}}
 {id=WavePlayer_1, properties {soundWaveAssetPtr=&/Emote_SFX_Rocking_Chair_Intro}}
 {id=WavePlayer_2, properties {soundWaveAssetPtr=&/Emote_MX_RockingChair_Intro}}
]
properties {
 maxDistance=2097152
 soundClassObject=/Game/Sounds/FortSoundClassesAndModes/Fort_Music_Emote_1P_PSM
 volumeMultiplier=1.2
 concurrencySet [/Game/Sounds/Concurrency_Rules/Athena_EmoteMusic_1P_Conc]
 soundSubmixSends [{minSendLevel=0, soundSubmix=&Fortnite_AudioAnalysis_Music}]
}
